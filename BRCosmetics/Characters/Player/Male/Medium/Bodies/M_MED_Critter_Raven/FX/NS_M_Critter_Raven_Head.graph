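# Niagara head effect for the Raven critter player cosmetic, adding glowing eyes, a head shadow and feathers shed while jumping.
# EyeGlow places two purple sprites with lights on the skeletal mesh, mirrored by spawn group, and HeadShadow adds one attached black sprite.
# Feathers spawn from the mesh at a user-scaled rate, then fall with drag, curl noise and sine-driven flipping.
# User.bisHeadHidden sets every emitter's VisibilityTag and disables the eye lights.
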NiagaraSystem /BRCosmetics/Characters/Player/Male/Medium/Bodies/M_MED_Critter_Raven/FX/NS_M_Critter_Raven_Head
dependencies {
 usedDataInterfaces [&NiagaraDataInterfaceCurve, &NiagaraDataInterfaceVector2DCurve]
 usedDynamicInputs []
 usedModules [&SystemState]
}
emitters {
 EyeGlow {
  emitterUpdateScript [
   &EmitterState(
    Inactive Response=::NewEnumerator1
    Life Cycle Mode=::NewEnumerator1
    Loop Behavior=::NewEnumerator1
    Loop Duration Mode=::NewEnumerator1
    UseLoopDelay=false
   )
   &SpawnBurst_Instantaneous(Spawn Count=2, Use Spawn Probability=false, UseLoopCountLimit=false)
  ]
  particleSpawnScript [
   &FN_SetSpawnGroup(Number of Spawn Groups=2)
   &InitializeParticle(
    Color=<rgba 0.1728 0 0.4 1>
    Sprite Rotation Angle=&MaskFloatBySpawnGroup(Masked Float=-24, Passthrough Float=24)
    Sprite Rotation Mode=::NewEnumerator0
    Sprite Size=<10 7>
    Sprite Size Mode=::NewEnumerator4
    UsePositionOffset=false
    Write Material Random=true
   )
  ]
  particleUpdateScript [
   &CameraOffset(Camera Offset Amount=5)
   &SkeletalMeshLocation(
    Bone Sampling Mode=::NewEnumerator4
    MaskSpawnGroup=false
    Sampled Position Offset=&Lerp_Vector(A=<3.6 9.5 -4.3>, B=<3.6 9.5 4.3>, Alpha=&MakeFloatFromInt(Int=$Particles.SpawnGroup))
    Skeletal Mesh=$User.SkeletalMesh
   )
   &ParticleState(
    Kill Particles When Lifetime Has Elapsed=false
    Let Infinitely Lived Particles Die When Emitter Deactivates=true
    Loop Particles Lifetime=true
   )
   &FN_Light_Attributes(
    Color=$Particles.Initial.Color
    Create Color Attribute=true
    Create Exponent Attribute=true
    Create Radius Attribute=true
    Enable Particle Light Enable=true
    Exponent=2
    Light Enabled:NiagaraBool=&Boolean_NotOperation(Bool=$User.bisHeadHidden)
    Radius=15
    Scale Color=6
   )
   /SetVariables(Particles.VisibilityTag=&MakeCustomIntFromBool(Bool=$User.bisHeadHidden, True INT=1, False INT=0)) {isSetParameters=true, module=SetVariables_#1}
  ]
  rendererClasses [&NiagaraSpriteRendererProperties, &NiagaraLightRendererProperties]
  renderers [
   {rendererClass=&NiagaraSpriteRendererProperties}
   {rendererClass=&NiagaraLightRendererProperties, rendererIndex=1}
  ]
 }
 Feathers {
  emitterUpdateScript [
   &EmitterState()
   &SpawnRate(
    SpawnRate=&MakeCustomFloatFromBool(Bool=$User.IsJumping, True Float=&Multiply_Float(A=6, B=$User.FeatherSpawnScale), False Float=0)
    Use Spawn Probability=false
   )
  ]
  particleSpawnScript [
   &InitializeParticle(
    Color Mode=::NewEnumerator0
    Lifetime Min=1.5
    Lifetime Mode=::NewEnumerator1
    Sprite Rotation Angle=&RandomRangeFloat(Minimum=0, Maximum=1)
    Sprite Rotation Mode=::NewEnumerator1
    Sprite Size Mode=::NewEnumerator3
    Uniform Sprite Size=18
    UsePositionOffset=false
    Write Material Random=true
   )
   &SkeletalMeshLocation(Bone Sampling Mode=::NewEnumerator0, MaskSpawnGroup=false, OverrideRandomSeed=false, Skeletal Mesh=$User.SkeletalMesh)
   &FN_SetSpawnGroup(Normalize Spawn Group=true, Number of Spawn Groups=6, Particles Per Group=1)
   &AddVelocity(Velocity=&RandomRangeVector(Minimum=<100 -100 0>, Maximum=<-100 -150 0>, Randomness Mode=::NewEnumerator1, Random Seed=12))
  ]
  particleUpdateScript [
   &ParticleState()
   &GravityForce(Gravity=<0 0 -200>)
   &Drag(Drag=2, Use Linear Drag=true, Use Rotational Drag=false)
   &CurlNoiseForce(Noise Frequency=25, Noise Strength=350, Pan Noise=true, Pan Noise Field=<0 0 -1>)
   &SolveForcesAndVelocity(Clamp Velocity=false, Limit Acceleration=false)
   &SpriteRotationRate(Rotation Rate=&RandomRangeFloat(Minimum=100, Maximum=200))
   &ScaleSpriteSize(
    Non-Uniform Curve Scale=&MakeVector2D(
     X=1
     Y=&Sine(Normalized Angle=$Emitter.Age, Period=&RandomRangeFloat(Minimum=2, Maximum=1), Scale=1, Bias=0.1)
    )
    Non-Uniform Curve Sprite Scale=DataInterface() {propertyValues {
      XCurve {keys [
        {arriveTangent=8.2479, interpMode=RCIM_Cubic, leaveTangent=8.2479, tangentMode=RCTM_User, time=0, value=0.5}
        <at 0.05 1>
        {interpMode=RCIM_Cubic, tangentMode=RCTM_User, time=0.7, value=1}
        {tangentMode=RCTM_User, time=1, value=0}
       ]}
      YCurve {keys [
        {arriveTangent=8.2479, interpMode=RCIM_Cubic, leaveTangent=8.2479, tangentMode=RCTM_User, time=0, value=0.5}
        <at 0.05 1>
        {interpMode=RCIM_Cubic, tangentMode=RCTM_User, time=0.7, value=1}
        {tangentMode=RCTM_User, time=1, value=0}
       ]}
     }}
    Scale Sprite Size Mode=::NewEnumerator3
   )
   /SetVariables(Particles.VisibilityTag=&MakeCustomIntFromBool(Bool=$User.bisHeadHidden, True INT=1, False INT=0)) {isSetParameters=true, module=SetVariables_#1}
  ]
  rendererClasses [&NiagaraSpriteRendererProperties]
  renderers [{rendererClass=&NiagaraSpriteRendererProperties}]
 }
 HeadShadow {
  emitterUpdateScript [
   &EmitterState(
    Inactive Response=::NewEnumerator1
    Life Cycle Mode=::NewEnumerator1
    Loop Behavior=::NewEnumerator1
    Loop Duration Mode=::NewEnumerator1
    UseLoopDelay=false
   )
   &SpawnBurst_Instantaneous(Use Spawn Probability=false, UseLoopCountLimit=false)
  ]
  particleSpawnScript [
   &FN_SetSpawnGroup(Number of Spawn Groups=1)
   &InitializeParticle(Color=<rgba 0 0 0 1>, Sprite Size Mode=::NewEnumerator3, Uniform Sprite Size=24, UsePositionOffset=false, Write Material Random=true)
  ]
  particleUpdateScript [
   &SkeletalMeshLocation(Bone Sampling Mode=::NewEnumerator4, MaskSpawnGroup=false, Sampled Position Offset=<3 4.5 0>, Skeletal Mesh=$User.SkeletalMesh)
   &ParticleState(
    Kill Particles When Lifetime Has Elapsed=false
    Let Infinitely Lived Particles Die When Emitter Deactivates=true
    Loop Particles Lifetime=true
   )
   /SetVariables(Particles.VisibilityTag=&MakeCustomIntFromBool(Bool=$User.bisHeadHidden, True INT=1, False INT=0)) {isSetParameters=true, module=SetVariables_#1}
  ]
  rendererClasses [&NiagaraSpriteRendererProperties]
  renderers [{rendererClass=&NiagaraSpriteRendererProperties}]
 }
}
summary {
 systemName=NS_M_Critter_Raven_Head
 emitters [
  {emitterName=EyeGlow, rendererClasses [&NiagaraSpriteRendererProperties, &NiagaraLightRendererProperties]}
  {emitterName=HeadShadow, rendererClasses [&NiagaraSpriteRendererProperties]}
  {emitterName=Feathers, rendererClasses [&NiagaraSpriteRendererProperties]}
 ]
 userVariables [
  User.bisHeadHidden {type=NiagaraBool}
  User.FeatherSpawnScale {type=NiagaraFloat, defaultValue=1, description=""}
  User.IsJumping {type=NiagaraBool}
  User.SkeletalMesh {
   type=NiagaraDataInterfaceSkeletalMesh
   defaultValue {
    struct=NiagaraExt_VariableValue_DataInterface
    value {dataInterface=&self:NiagaraDataInterfaceSkeletalMesh_0}
   }
  }
 ]
}
