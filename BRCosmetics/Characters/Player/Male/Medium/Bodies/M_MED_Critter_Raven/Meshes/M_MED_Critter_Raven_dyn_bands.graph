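# Physics asset for the dangling right-arm strap on the male medium Critter Raven character body.
# Two capsule strap segments chain from a small master body attached to the right forearm, with limited swing that tightens toward the tip and almost no twist.
# Larger capsules on the right forearm, upper arm, thigh and pelvis approximate the body, linked by angularly free constraints, as is the master-to-forearm joint.
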PhysicsAsset /BRCosmetics/Characters/Player/Male/Medium/Bodies/M_MED_Critter_Raven/Meshes/M_MED_Critter_Raven_dyn_bands
bodies {
 dyn_arm_strap_r_01 [{shapeType=Capsule, length=2.5, radius=1.5, shapeName=dyn_arm_strap_r_01_capsule, center=<4.0699 0 -0.0001>, rotation=<rot 90 0 0>}]
 dyn_arm_strap_r_02 [{shapeType=Capsule, length=5, radius=1.5, shapeName=dyn_arm_strap_r_02_capsule, center=<3 0 0>, rotation=<rot 90 0 0>}]
 dyn_arm_strap_r_MSTR [{shapeType=Capsule, length=0.505, radius=0.505, shapeName=dyn_skirt_bk_l_MSTR_capsule, rotation=<rot 0 -90 0>}]
 lowerarm_r [{shapeType=Capsule, length=19, radius=4.75, shapeName=thigh_twist_01_l_capsule, center=<-10.1645 0 0.0001>, rotation=<rot 90 0 2.0707>}]
 pelvis [{
   shapeType=Capsule
   length=6.6306
   radius=16.3159
   shapeName=pelvis_capsule
   center=<-2.2448 -1.5243 -0.2971>
   rotation=<rot 3.9757 -90 1.9786>
  }]
 thigh_r [{
   shapeType=Capsule
   length=17
   radius=16
   shapeName=thigh_twist_01_l_capsule
   center=<12.7772 2.1286 -1.1181>
   rotation=<rot 79.9974 -174.9998 180>
  }]
 upperarm_r [{shapeType=Capsule, length=19, radius=4.75, shapeName=thigh_twist_01_l_capsule, center=<-13.5932 0.1239 0.0001>, rotation=<rot 90 0 2.0707>}]
}
constraints [
 {bone1Name=dyn_arm_strap_r_01, bone2Name=dyn_arm_strap_r_MSTR, swing1LimitDegrees=65, swing2LimitDegrees=65, twistLimitDegrees=2}
 {bone1Name=dyn_arm_strap_r_02, bone2Name=dyn_arm_strap_r_01, swing1LimitDegrees=45, swing2LimitDegrees=45, twistLimitDegrees=2}
 {
  bone1Name=dyn_arm_strap_r_MSTR
  bone2Name=lowerarm_r
  swing1LimitDegrees=45
  swing1Motion=Free
  swing2LimitDegrees=45
  swing2Motion=Free
  twistLimitDegrees=45
  twistMotion=Free
 }
 {
  bone1Name=lowerarm_r
  bone2Name=upperarm_r
  swing1LimitDegrees=45
  swing1Motion=Free
  swing2LimitDegrees=45
  swing2Motion=Free
  twistLimitDegrees=45
  twistMotion=Free
 }
 {
  bone1Name=upperarm_r
  bone2Name=pelvis
  swing1LimitDegrees=45
  swing1Motion=Free
  swing2LimitDegrees=45
  swing2Motion=Free
  twistLimitDegrees=45
  twistMotion=Free
 }
 {
  bone1Name=thigh_r
  bone2Name=pelvis
  swing1LimitDegrees=45
  swing1Motion=Free
  swing2LimitDegrees=45
  swing2Motion=Free
  twistLimitDegrees=45
  twistMotion=Free
 }
]
