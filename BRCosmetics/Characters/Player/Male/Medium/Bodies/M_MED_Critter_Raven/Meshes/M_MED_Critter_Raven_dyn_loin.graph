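# Physics asset for the dynamic loin belt on the male medium Critter Raven character body, built entirely from capsule bodies.
# A constraint chain runs from dyn_loin_MSTR through the belt base to three belt segments, with free angular motion at the root, moderate limits on the first segment and tight limits on the last two.
# Unconstrained capsules on the thighs, calves and a large belt TGT body supply collision volumes around the chain.
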PhysicsAsset /BRCosmetics/Characters/Player/Male/Medium/Bodies/M_MED_Critter_Raven/Meshes/M_MED_Critter_Raven_dyn_loin
bodies {
 calf_l [{
   shapeType=Capsule
   length=37.5
   radius=8
   shapeName=thigh_twist_01_l_capsule
   center=<-24.3267 -2.4035 0.5327>
   rotation=<rot 90 12.9074 7.9075>
  }]
 calf_r [{
   shapeType=Capsule
   length=40.5
   radius=8
   shapeName=thigh_twist_01_l_capsule
   center=<26.3595 2.4035 -0.5325>
   rotation=<rot 90 3.3723 -176.6276>
  }]
 dyn_loin_MSTR [{shapeType=Capsule, length=1, radius=1, shapeName=dyn_loin_MSTR_capsule}]
 dyn_loin_belt_01 [{shapeType=Capsule, length=2, radius=6.75, shapeName=dyn_loin_belt_01_capsule, center=<5 0 3.5>, rotation=<rot 90 0 0>}]
 dyn_loin_belt_02 [{shapeType=Capsule, length=4.5, radius=6.75, shapeName=dyn_loin_belt_02_capsule, center=<5 0 3.5>, rotation=<rot 90 0 0>}]
 dyn_loin_belt_03 [{shapeType=Capsule, length=4.5, radius=6.75, shapeName=dyn_loin_belt_03_capsule, center=<8 0 3.5>, rotation=<rot 90 0 0>}]
 dyn_loin_belt_TGT [{shapeType=Capsule, length=23.5, radius=26.5, shapeName=thigh_twist_01_l_capsule, center=<-29.5997 -0.0008 -30.2773>, rotation=<rot 0 0 90>}]
 dyn_loin_belt_base [{shapeType=Capsule, length=1, radius=1, shapeName=dyn_loin_belt_base_capsule}]
 thigh_twist_01_l [{shapeType=Capsule, length=23, radius=17.5, shapeName=thigh_twist_01_l_capsule, center=<-5.837 -8.8574 -2.6624>, rotation=<rot 90 90 92.4338>}]
 thigh_twist_01_r [{
   shapeType=Capsule
   length=23
   radius=17.5
   shapeName=thigh_twist_01_l_capsule
   center=<5.837 8.8574 2.6624>
   rotation=<rot 90 -18.166 164.2678>
  }]
}
constraints [
 {
  bone1Name=dyn_loin_belt_base
  bone2Name=dyn_loin_MSTR
  swing1LimitDegrees=45
  swing1Motion=Free
  swing2LimitDegrees=45
  swing2Motion=Free
  twistLimitDegrees=45
  twistMotion=Free
 }
 {bone1Name=dyn_loin_belt_01, bone2Name=dyn_loin_belt_base, swing1LimitDegrees=45, swing2LimitDegrees=45, twistLimitDegrees=65}
 {bone1Name=dyn_loin_belt_02, bone2Name=dyn_loin_belt_01, swing1LimitDegrees=5, swing2LimitDegrees=2, twistLimitDegrees=10}
 {bone1Name=dyn_loin_belt_03, bone2Name=dyn_loin_belt_02, swing1LimitDegrees=5, swing2LimitDegrees=2, twistLimitDegrees=10}
]
