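# Physics asset for the male medium Raven critter body, setting up secondary motion for its dynamic skirt, loin belt and right arm strap.
# Capsules cover those dynamic bones plus thighs, calves and right arm.
# Master bones join thighs or forearm with free constraints, anchoring swing-limited chains with tight twist; the belt instead allows wide twist from its own base.
# A free constraint also links right side and back skirt tips, with no left counterpart.
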PhysicsAsset /BRCosmetics/Characters/Player/Male/Medium/Bodies/M_MED_Critter_Raven/Meshes/M_MED_Critter_Raven_dyn_skirt
bodies {
 calf_l [{
   shapeType=Capsule
   length=37.5
   radius=8
   shapeName=thigh_twist_01_l_capsule
   center=<-24.3267 -2.4035 0.5327>
   rotation=<rot 90 12.9074 7.9075>
  }]
 calf_r [{
   shapeType=Capsule
   length=40.5
   radius=8
   shapeName=thigh_twist_01_l_capsule
   center=<26.3595 2.4035 -0.5325>
   rotation=<rot 90 3.3723 -176.6276>
  }]
 dyn_arm_strap_r_01 [{shapeType=Capsule, length=2.5, radius=1.5, shapeName=dyn_arm_strap_r_01_capsule, center=<4.0699 0 -0.0001>, rotation=<rot 90 0 0>}]
 dyn_arm_strap_r_02 [{shapeType=Capsule, length=5, radius=1.5, shapeName=dyn_arm_strap_r_02_capsule, center=<3 0 0>, rotation=<rot 90 0 0>}]
 dyn_arm_strap_r_MSTR [{shapeType=Capsule, length=0.505, radius=0.505, shapeName=dyn_skirt_bk_l_MSTR_capsule, rotation=<rot 0 -90 0>}]
 dyn_loin_belt [{shapeType=Capsule, length=55.5, radius=17.25, shapeName=dyn_loin_belt_capsule, center=<29.9995 0 16>, rotation=<rot 0 0 90>}]
 dyn_loin_belt_base [{shapeType=Capsule, length=1, radius=1, shapeName=dyn_loin_MSTR_capsule}]
 dyn_skirt_bk_l_01 [{shapeType=Capsule, length=27.75, radius=7, shapeName=dyn_skirt_bk_l_01_capsule, center=<8 0 2>, rotation=<rot 0 0 90>}]
 dyn_skirt_bk_l_02 [{shapeType=Capsule, length=27.75, radius=7, shapeName=dyn_skirt_bk_l_02_capsule, center=<8 0 2>, rotation=<rot 0 0 90>}]
 dyn_skirt_bk_l_03 [{shapeType=Capsule, length=27.75, radius=7, shapeName=dyn_skirt_bk_l_03_capsule, center=<8 0 2>, rotation=<rot 0 0 90>}]
 dyn_skirt_bk_l_MSTR [{shapeType=Capsule, length=0.505, radius=0.505, shapeName=dyn_skirt_bk_l_MSTR_capsule, rotation=<rot 0 -90 0>}]
 dyn_skirt_bk_r_01 [{shapeType=Capsule, length=27.75, radius=7, shapeName=dyn_skirt_bk_r_01_capsule, center=<8 0 2>, rotation=<rot 0 0 90>}]
 dyn_skirt_bk_r_02 [{shapeType=Capsule, length=27.75, radius=7, shapeName=dyn_skirt_bk_r_02_capsule, center=<8 0 2>, rotation=<rot 0 0 90>}]
 dyn_skirt_bk_r_03 [{shapeType=Capsule, length=27.75, radius=7, shapeName=dyn_skirt_bk_r_03_capsule, center=<8 0 2>, rotation=<rot 0 0 90>}]
 dyn_skirt_bk_r_MSTR [{shapeType=Capsule, length=0.505, radius=0.505, shapeName=dyn_skirt_bk_r_MSTR_capsule, rotation=<rot 0 -90 0>}]
 dyn_skirt_l_01 [{shapeType=Capsule, length=11.25, radius=8.5, shapeName=dyn_skirt_bk_l_01_capsule, center=<9 -5 5>, rotation=<rot 0 0 74.9984>}]
 dyn_skirt_l_02 [{shapeType=Capsule, length=11.25, radius=8.5, shapeName=dyn_skirt_bk_l_01_capsule, center=<9 -5 3>, rotation=<rot 0 0 74.9988>}]
 dyn_skirt_l_MSTR [{shapeType=Capsule, length=0.505, radius=0.505, shapeName=dyn_skirt_l_MSTR_capsule, rotation=<rot 0 -90 0>}]
 dyn_skirt_r_01 [{shapeType=Capsule, length=11.25, radius=8.5, shapeName=dyn_skirt_bk_l_01_capsule, center=<9 5 5>, rotation=<rot 0 0 110.0005>}]
 dyn_skirt_r_02 [{shapeType=Capsule, length=11.25, radius=8.5, shapeName=dyn_skirt_bk_l_01_capsule, center=<9 5 3>, rotation=<rot 0 0 109.9999>}]
 dyn_skirt_r_MSTR [{shapeType=Capsule, length=0.505, radius=0.505, shapeName=dyn_skirt_r_MSTR_capsule, rotation=<rot 0 -90 0>}]
 lowerarm_r [{shapeType=Capsule, length=19, radius=4.75, shapeName=thigh_twist_01_l_capsule, center=<-10.1645 0 0.0001>, rotation=<rot 90 0 2.0707>}]
 thigh_twist_01_l [{
   shapeType=Capsule
   length=33.75
   radius=9.5
   shapeName=thigh_twist_01_l_capsule
   center=<-4.1195 -0.636 1.4396>
   rotation=<rot 90 12.9074 7.9075>
  }]
 thigh_twist_01_r [{shapeType=Capsule, length=33.75, radius=9.5, shapeName=thigh_twist_01_l_capsule, center=<4.1195 0.636 -1.4396>, rotation=<rot 90 -90 85>}]
 upperarm_r [{shapeType=Capsule, length=19, radius=4.75, shapeName=thigh_twist_01_l_capsule, center=<-13.5932 0.1239 0.0001>, rotation=<rot 90 0 2.0707>}]
}
constraints [
 {
  bone1Name=dyn_skirt_bk_l_MSTR
  bone2Name=thigh_twist_01_l
  swing1LimitDegrees=45
  swing1Motion=Free
  swing2LimitDegrees=45
  swing2Motion=Free
  twistLimitDegrees=45
  twistMotion=Free
 }
 {bone1Name=dyn_skirt_bk_l_01, bone2Name=dyn_skirt_bk_l_MSTR, swing1LimitDegrees=35, swing2LimitDegrees=15, twistLimitDegrees=2}
 {bone1Name=dyn_skirt_bk_l_02, bone2Name=dyn_skirt_bk_l_01, swing1LimitDegrees=35, swing2LimitDegrees=35, twistLimitDegrees=2}
 {bone1Name=dyn_skirt_bk_l_03, bone2Name=dyn_skirt_bk_l_02, swing1LimitDegrees=35, swing2LimitDegrees=35, twistLimitDegrees=2}
 {
  bone1Name=dyn_skirt_l_MSTR
  bone2Name=thigh_twist_01_l
  swing1LimitDegrees=45
  swing1Motion=Free
  swing2LimitDegrees=45
  swing2Motion=Free
  twistLimitDegrees=45
  twistMotion=Free
 }
 {bone1Name=dyn_skirt_l_01, bone2Name=dyn_skirt_l_MSTR, swing1LimitDegrees=65, swing2LimitDegrees=15, twistLimitDegrees=2}
 {bone1Name=dyn_skirt_bk_r_01, bone2Name=dyn_skirt_bk_r_MSTR, swing1LimitDegrees=35, swing2LimitDegrees=15, twistLimitDegrees=2}
 {bone1Name=dyn_skirt_bk_r_02, bone2Name=dyn_skirt_bk_r_01, swing1LimitDegrees=35, swing2LimitDegrees=35, twistLimitDegrees=2}
 {bone1Name=dyn_skirt_bk_r_03, bone2Name=dyn_skirt_bk_r_02, swing1LimitDegrees=35, swing2LimitDegrees=35, twistLimitDegrees=2}
 {bone1Name=dyn_skirt_r_01, bone2Name=dyn_skirt_r_MSTR, swing1LimitDegrees=45, swing2LimitDegrees=15, twistLimitDegrees=2}
 {
  bone1Name=dyn_skirt_bk_r_MSTR
  bone2Name=thigh_twist_01_r
  swing1LimitDegrees=45
  swing1Motion=Free
  swing2LimitDegrees=45
  swing2Motion=Free
  twistLimitDegrees=45
  twistMotion=Free
 }
 {
  bone1Name=dyn_skirt_r_MSTR
  bone2Name=thigh_twist_01_r
  swing1LimitDegrees=45
  swing1Motion=Free
  swing2LimitDegrees=45
  swing2Motion=Free
  twistLimitDegrees=45
  twistMotion=Free
 }
 {bone1Name=dyn_skirt_r_02, bone2Name=dyn_skirt_r_01, swing1LimitDegrees=45, swing2LimitDegrees=15, twistLimitDegrees=2}
 {bone1Name=dyn_skirt_l_02, bone2Name=dyn_skirt_l_01, swing1LimitDegrees=65, swing2LimitDegrees=15, twistLimitDegrees=2}
 {
  bone1Name=dyn_skirt_r_02
  bone2Name=dyn_skirt_bk_r_03
  swing1LimitDegrees=45
  swing1Motion=Free
  swing2LimitDegrees=45
  swing2Motion=Free
  twistLimitDegrees=45
  twistMotion=Free
 }
 {bone1Name=dyn_arm_strap_r_01, bone2Name=dyn_arm_strap_r_MSTR, swing1LimitDegrees=65, swing2LimitDegrees=65, twistLimitDegrees=2}
 {bone1Name=dyn_arm_strap_r_02, bone2Name=dyn_arm_strap_r_01, swing1LimitDegrees=45, swing2LimitDegrees=45, twistLimitDegrees=2}
 {
  bone1Name=dyn_arm_strap_r_MSTR
  bone2Name=lowerarm_r
  swing1LimitDegrees=45
  swing1Motion=Free
  swing2LimitDegrees=45
  swing2Motion=Free
  twistLimitDegrees=45
  twistMotion=Free
 }
 {
  bone1Name=lowerarm_r
  bone2Name=upperarm_r
  swing1LimitDegrees=45
  swing1Motion=Free
  swing2LimitDegrees=45
  swing2Motion=Free
  twistLimitDegrees=45
  twistMotion=Free
 }
 {bone1Name=dyn_loin_belt, bone2Name=dyn_loin_belt_base, swing1LimitDegrees=5, swing2LimitDegrees=30, twistLimitDegrees=65}
]
