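# Post-copy AnimBlueprint for the Raven character body cosmetic, adding secondary motion to skirts, loin belt, straps, bands and ornaments.
# It copies the parent mesh pose, then layers weighted constraints, a look-at, Anim Dynamics straps and three RigidBody simulations.
# Each update maps thigh and upper-arm rotation to alphas for corrective bone transforms.
# A deprecated gravity-state function feeds the RigidBody gravity override.
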AnimBlueprint /BRCosmetics/Characters/Player/Male/Medium/Bodies/M_MED_Critter_Raven/Meshes/M_MED_Critter_Raven_AnimBP
target_skeleton=&/M_MED_Critter_Raven_Skeleton
graphs [
 AnimGraph {nodes [
   Anim Dynamics - Bone: dyn_arm_strap_r_01(Alpha=1, ExternalForce="0,0,0") {
    angularDampingOverride=0.999
    angularSpringConstant=1000
    bAngularSpring=true
    bGravityOverrideInSimSpace=true
    bOverrideAngularDamping=true
    bUseGravityOverride=true
    simulationSpace=RootRelative
    alphaBoolBlend {blendOption=Linear}
    alphaScaleBiasClamp {scale=0.6}
    boundBone {boneName=dyn_arm_strap_r_01}
    chainEnd {boneName=dyn_arm_strap_r_02}
    componentAppliedLinearAccClamp {z=5000}
    componentLinearAccScale=<0.5 0.5 0.5>
    componentLinearVelScale=<0.5 0.5 0.5>
    gravityOverride {z=-980}
    physicsBodyDefinitions [{
      sphereCollisionRadius=0
      boundBone {boneName=dyn_arm_strap_r_01}
      boxExtents=<10 20 20>
      constraintSetup {angularLimitsMax=<0 90 90>, angularLimitsMin=<0 -90 -125>, angularTarget=<0 0 0>}
      localJointOffset=<5 0 0>
     }]
   }
   "Transform (Modify) Bone - Bone: dyn_skirt_bk_r_MSTR"(Alpha=1, Translation="-20,0,0", Rotation="0,0.000,-15", Scale="1,1,1") {
    rotationMode=BMM_Additive
    rotationSpace=BCS_BoneSpace
    translationMode=BMM_Additive
    translationSpace=BCS_BoneSpace
    alphaBoolBlend {blendOption=Linear}
    boneToModify {boneName=dyn_skirt_bk_r_MSTR}
    rotation {roll=-15}
    translation {x=-20}
   }
   "Transform (Modify) Bone - Bone: dyn_skirt_bk_l_MSTR"(Alpha=1, Translation="-15,0,0", Rotation="0,0.000,15", Scale="1,1,1") {
    rotationMode=BMM_Additive
    rotationSpace=BCS_BoneSpace
    translationMode=BMM_Additive
    translationSpace=BCS_BoneSpace
    alphaBoolBlend {blendOption=Linear}
    boneToModify {boneName=dyn_skirt_bk_l_MSTR}
    rotation {roll=15}
    translation {x=-15}
   }
   n1 = "Constraint - dyn_knie_MSTR (PP)"!(
    ComponentPose=Local To Component(LocalPose=Copy Pose From Mesh() {bUseAttachedParent=true})
    Alpha=1
    ConstraintWeights_0=0.8151
    ConstraintWeights_1:float=0.1849
   ) -> (Pose) {
    alphaBoolBlend {blendOption=Linear}
    boneToModify {boneName=dyn_knie_MSTR}
    constraintSetup [{targetBone {boneName=calf_l}}, {targetBone {boneName=calf_twist_02_l}}]
    constraintWeights [1,1]
   }
   n2 = "Constraint - dyn_loin_MSTR (PPPPP)"!(
    ComponentPose=@n1
    Alpha=1
    ConstraintWeights_0=0.2106
    ConstraintWeights_1:float=0.3981
    ConstraintWeights_2:float=0.099
    ConstraintWeights_3:float=0.0162
    ConstraintWeights_4:float=0.276
   ) -> (Pose) {
    alphaBoolBlend {blendOption=Linear}
    boneToModify {boneName=dyn_loin_MSTR}
    constraintSetup [
     {targetBone {boneName=spine_01}}
     {targetBone {boneName=spine_02}}
     {targetBone {boneName=spine_03}}
     {targetBone {boneName=spine_04}}
     {targetBone {boneName=pelvis}}
    ]
    constraintWeights [1,1,1,1,1]
   }
   n3 = "Constraint - dyn_ornaments_r (PPPP)"!(
    ComponentPose=@n2
    Alpha=1
    ConstraintWeights_0=0.6579
    ConstraintWeights_1:float=0.0676
    ConstraintWeights_2:float=0.012
    ConstraintWeights_3:float=0.2625
   ) -> (Pose) {
    alphaBoolBlend {blendOption=Linear}
    boneToModify {boneName=dyn_ornaments_r}
    constraintSetup [
     {targetBone {boneName=upperarm_twist_01_r}}
     {targetBone {boneName=clavicle_r}}
     {targetBone {boneName=spine_04}}
     {targetBone {boneName=spine_05}}
    ]
    constraintWeights [1,1,1,1]
   }
   n4 = "Constraint - dyn_ornaments_l (PPPP)"!(
    ComponentPose=@n3
    Alpha=1
    ConstraintWeights_0=0.5572
    ConstraintWeights_1:float=0.012
    ConstraintWeights_2:float=0.2626
    ConstraintWeights_3:float=0.1682
   ) -> (Pose) {
    alphaBoolBlend {blendOption=Linear}
    boneToModify {boneName=dyn_ornaments_l}
    constraintSetup [
     {targetBone {boneName=upperarm_twist_01_l}}
     {targetBone {boneName=spine_04}}
     {targetBone {boneName=spine_05}}
     {targetBone {boneName=clavicle_l}}
    ]
    constraintWeights [1,1,1,1]
   }
   n5 = Get arm_rot_up_r() -> (arm_rot_up_r:double)
   n6 = "Transform (Modify) Bone - Bone: dyn_ornaments_r_MSTR"(ComponentPose=@n4, Alpha=@n5, Translation="2,0,0", Rotation="0,0,45", Scale="1,1,1") {
    rotationMode=BMM_Additive
    rotationSpace=BCS_BoneSpace
    translationMode=BMM_Additive
    translationSpace=BCS_BoneSpace
    alphaBoolBlend {blendOption=Linear}
    boneToModify {boneName=dyn_ornaments_r_MSTR}
    rotation {roll=45}
    translation {x=2}
   }
   n7 = "Transform (Modify) Bone - Bone: dyn_ornaments_r"(ComponentPose=@n6, Alpha=@n5, Translation="0,-1,0", Rotation="25,15,15", Scale="1,1,1") {
    rotationMode=BMM_Additive
    rotationSpace=BCS_BoneSpace
    translationMode=BMM_Additive
    translationSpace=BCS_BoneSpace
    alphaBoolBlend {blendOption=Linear}
    boneToModify {boneName=dyn_ornaments_r}
    rotation=<rot 25 15 15>
    translation {y=-1}
   }
   n8 = Get arm_rot_up_l() -> (arm_rot_up_l:double)
   n9 = "Transform (Modify) Bone - Bone: dyn_ornaments_l_MSTR"(ComponentPose=@n7, Alpha=@n8, Translation="-2,0,0", Rotation="0,0,-25", Scale="1,1,1") {
    rotationMode=BMM_Additive
    rotationSpace=BCS_BoneSpace
    translationMode=BMM_Additive
    translationSpace=BCS_BoneSpace
    alphaBoolBlend {blendOption=Linear}
    boneToModify {boneName=dyn_ornaments_l_MSTR}
    rotation {roll=-25}
    translation {x=-2}
   }
   n10 = "Transform (Modify) Bone - Bone: dyn_ornaments_l"(ComponentPose=@n9, Alpha=@n8, Translation="1,-1.5,0", Rotation="35,0,-15", Scale="1,1,1") {
    rotationMode=BMM_Additive
    rotationSpace=BCS_BoneSpace
    translationMode=BMM_Additive
    translationSpace=BCS_BoneSpace
    alphaBoolBlend {blendOption=Linear}
    boneToModify {boneName=dyn_ornaments_l}
    rotation {pitch=35, roll=-15}
    translation=<1 -1.5>
   }
   n11 = "Constraint - dyn_arm_strap_r_MSTR (PPPP)"!(
    ComponentPose=@n10
    Alpha=1
    ConstraintWeights_0=0.0164
    ConstraintWeights_1:float=0.1245
    ConstraintWeights_2:float=0.4928
    ConstraintWeights_3:float=0.3663
   ) -> (Pose) {
    alphaBoolBlend {blendOption=Linear}
    boneToModify {boneName=dyn_arm_strap_r_MSTR}
    constraintSetup [
     {targetBone {boneName=upperarm_twist_01_r}}
     {targetBone {boneName=upperarm_r}}
     {targetBone {boneName=lowerarm_r}}
     {targetBone {boneName=lowerarm_twist_01_r}}
    ]
    constraintWeights [1,1,1,1]
   }
   n12 = Anim Dynamics - Bone: dyn_waist_strap(ComponentPose=@n11, Alpha=1, ExternalForce="0,0,0") {
    angularDampingOverride=0.9999
    bAngularSpring=true
    bGravityOverrideInSimSpace=true
    bOverrideAngularDamping=true
    bUseGravityOverride=true
    simulationSpace=RootRelative
    alphaBoolBlend {blendOption=Linear}
    alphaScaleBiasClamp {scale=0.6}
    boundBone {boneName=dyn_waist_strap}
    componentAppliedLinearAccClamp {z=5000}
    componentLinearAccScale {z=0.5}
    componentLinearVelScale {z=0.5}
    gravityOverride {z=-980}
    physicsBodyDefinitions [{
      sphereCollisionRadius=0
      boundBone {boneName=dyn_waist_strap}
      boxExtents=<10 20 20>
      constraintSetup {angularLimitsMax=<0 0 0>, angularLimitsMin=<0 -65 0>, angularTarget=<0 0 0>}
      localJointOffset=<10 0 1>
     }]
   }
   n13 = Anim Dynamics - Bone: dyn_leg_strap_l(ComponentPose=@n12, Alpha=1, ExternalForce="0,0,0") {
    angularDampingOverride=0.9999
    angularSpringConstant=1000
    bAngularSpring=true
    bGravityOverrideInSimSpace=true
    bOverrideAngularDamping=true
    bUseGravityOverride=true
    simulationSpace=RootRelative
    alphaBoolBlend {blendOption=Linear}
    alphaScaleBiasClamp {scale=0.6}
    boundBone {boneName=dyn_leg_strap_l}
    componentAppliedLinearAccClamp {z=5000}
    componentLinearAccScale=<0.5 0.5 0.5>
    componentLinearVelScale=<0.5 0.5 0.5>
    gravityOverride {z=-980}
    physicsBodyDefinitions [{
      sphereCollisionRadius=0
      boundBone {boneName=dyn_leg_strap_l}
      boxExtents=<10 25 20>
      constraintSetup {angularLimitsMax=<0 25 35>, angularLimitsMin=<0 -65 -35>, angularTarget=<0 0 0>}
      localJointOffset=<5 0 0>
     }]
   }
   n14 = Anim Dynamics - Bone: dyn_leg_strap_r(ComponentPose=@n13, Alpha=1, ExternalForce="0,0,0") {
    angularDampingOverride=0.999
    angularSpringConstant=1000
    bAngularSpring=true
    bGravityOverrideInSimSpace=true
    bOverrideAngularDamping=true
    bUseGravityOverride=true
    simulationSpace=RootRelative
    alphaBoolBlend {blendOption=Linear}
    alphaScaleBiasClamp {scale=0.6}
    boundBone {boneName=dyn_leg_strap_r}
    componentAppliedLinearAccClamp {z=5000}
    componentLinearAccScale=<0.5 0.5 0.5>
    componentLinearVelScale=<0.5 0.5 0.5>
    gravityOverride {z=-980}
    physicsBodyDefinitions [{
      sphereCollisionRadius=0
      boundBone {boneName=dyn_leg_strap_r}
      boxExtents=<10 20 20>
      constraintSetup {angularLimitsMax=<0 35 45>, angularLimitsMin=<0 -45 -45>, angularTarget=<0 0 0>}
      localJointOffset=<5 0 0>
     }]
   }
   n15 = "Constraint - dyn_loin_belt_TGT (PPP)"!(ComponentPose=@n14, Alpha=1, ConstraintWeights_0=0, ConstraintWeights_1:float=0.5, ConstraintWeights_2:float=0.5) -> (Pose) {
    alphaBoolBlend {blendOption=Linear}
    boneToModify {boneName=dyn_loin_belt_TGT}
    constraintSetup [{targetBone {boneName=pelvis}}, {targetBone {boneName=thigh_twist_01_l}}, {targetBone {boneName=thigh_twist_01_r}}]
    constraintWeights [1,1,1]
   }
   n16 = Look At - Bone: dyn_loin_belt_base(ComponentPose=@n15, Alpha=1) {
    alphaBoolBlend {blendOption=Linear}
    boneToModify {boneName=dyn_loin_belt_base}
    lookAtLocation {x=0}
    lookAtTarget {boneReference {boneName=dyn_loin_belt_TGT}}
    lookAt_Axis {axis=<1 0>}
    lookUp_Axis {axis=<0 1>}
   }
   n17 = "Transform (Modify) Bone - Bone: dyn_loin_belt_base"(ComponentPose=@n16, Alpha=Get thigh_rot_up_l() -> (thigh_rot_up_l:double), Translation="1,0,0", Rotation="0,45,25", Scale="1,1,1") {
    rotationMode=BMM_Additive
    rotationSpace=BCS_BoneSpace
    scaleMode=BMM_Additive
    scaleSpace=BCS_BoneSpace
    translationMode=BMM_Additive
    translationSpace=BCS_BoneSpace
    alphaBoolBlend {blendOption=Linear}
    boneToModify {boneName=dyn_loin_belt_base}
    rotation {roll=25, yaw=45}
    translation {x=1}
   }
   n18 = "Transform (Modify) Bone - Bone: dyn_loin_belt_base"(ComponentPose=@n17, Alpha=Get thigh_rot_up_r() -> (thigh_rot_up_r:double), Translation="1,-2,0", Rotation="0,0,-15", Scale="1,1,1") {
    rotationMode=BMM_Additive
    rotationSpace=BCS_BoneSpace
    scaleMode=BMM_Additive
    scaleSpace=BCS_BoneSpace
    translationMode=BMM_Additive
    translationSpace=BCS_BoneSpace
    alphaBoolBlend {blendOption=Linear}
    boneToModify {boneName=dyn_loin_belt_base}
    rotation {roll=-15}
    translation=<1 -2>
   }
   n19 = "Constraint - dyn_skirt_bk_l_MSTR (P)"(ComponentPose=@n18, Alpha=1, ConstraintWeights_0=1) {
    alphaBoolBlend {blendOption=Linear}
    boneToModify {boneName=dyn_skirt_bk_l_MSTR}
    constraintSetup [{targetBone {boneName=pelvis}}]
    constraintWeights [1]
   }
   n20 = "Constraint - dyn_skirt_bk_r_MSTR (P)"(ComponentPose=@n19, Alpha=1, ConstraintWeights_0=1) {
    alphaBoolBlend {blendOption=Linear}
    boneToModify {boneName=dyn_skirt_bk_r_MSTR}
    constraintSetup [{targetBone {boneName=pelvis}}]
    constraintWeights [1]
   }
   n21 = "Transform (Modify) Bone - Bone: dyn_skirt_l_MSTR"(ComponentPose=@n20, Alpha=Get thigh_rot_up_l(), Translation="-10,-2.000,1.000", Rotation="0.000,-30,-10", Scale="1,1,1") {
    rotationMode=BMM_Additive
    rotationSpace=BCS_BoneSpace
    translationMode=BMM_Additive
    translationSpace=BCS_BoneSpace
    alphaBoolBlend {blendOption=Linear}
    boneToModify {boneName=dyn_skirt_l_MSTR}
    rotation {roll=-10, yaw=-30}
    translation=<-10 -2 1>
   }
   n22 = "Transform (Modify) Bone - Bone: dyn_skirt_r_MSTR"(ComponentPose=@n21, Alpha=Get thigh_rot_up_r(), Translation="-10,2,1", Rotation="0.000,0,10", Scale="1,1,1") {
    rotationMode=BMM_Additive
    rotationSpace=BCS_BoneSpace
    translationMode=BMM_Additive
    translationSpace=BCS_BoneSpace
    alphaBoolBlend {blendOption=Linear}
    boneToModify {boneName=dyn_skirt_r_MSTR}
    rotation {roll=10}
    translation=<-10 2 1>
   }
   n23 = Get GravityOverride()
   n24 = RigidBody!(ComponentPose=@n22, Alpha, OverrideWorldGravity:Vector=@n23, ExternalForce) -> (Pose) {
    bClampLinearTranslationLimitToRefPose=true
    bOverrideWorldGravity=true
    overridePhysicsAsset=&/M_MED_Critter_Raven_dyn_skirt
    simulationSpace=BaseBoneSpace
    alphaBoolBlend {blendOption=Linear}
    baseBoneRef {boneName=spine_02}
    componentAppliedLinearAccClamp=<3000 3000 3000>
    componentLinearAccScale {x=0.45}
    componentLinearVelScale {x=0.45}
    overrideWorldGravity {z=-980}
    simSpaceSettings {
     maxAngularAcceleration=200
     maxAngularVelocity=5
     maxLinearAcceleration=1000
     maxLinearVelocity=200
     velocityScaleZ=0
     worldAlpha=1
     externalLinearDragV=<1 1>
    }
   }
   n25 = RigidBody(ComponentPose=@n24, OverrideWorldGravity=@n23) {
    bClampLinearTranslationLimitToRefPose=true
    bOverrideWorldGravity=true
    overridePhysicsAsset=&/M_MED_Critter_Raven_dyn_loin
    simulationSpace=BaseBoneSpace
    alphaBoolBlend {blendOption=Linear}
    baseBoneRef {boneName=spine_01}
    componentAppliedLinearAccClamp=<3000 3000 3000>
    overrideWorldGravity {z=-980}
    simSpaceSettings {
     maxAngularAcceleration=200
     maxAngularVelocity=5
     maxLinearAcceleration=1000
     maxLinearVelocity=200
     velocityScaleZ=0
     externalLinearDragV=<1 1>
    }
   }
   n26 = RigidBody(ComponentPose=@n25, OverrideWorldGravity=@n23) {
    bClampLinearTranslationLimitToRefPose=true
    bOverrideWorldGravity=true
    overridePhysicsAsset=&/M_MED_Critter_Raven_dyn_bands
    simulationSpace=BaseBoneSpace
    alphaBoolBlend {blendOption=Linear}
    baseBoneRef {boneName=upperarm_r}
    componentAppliedLinearAccClamp=<3000 3000 3000>
    componentLinearAccScale=<0.3 0.3 0.3>
    componentLinearVelScale=<0.3 0.3 0.3>
    overrideWorldGravity {z=-980}
    simSpaceSettings {
     maxAngularAcceleration=200
     maxAngularVelocity=5
     maxLinearAcceleration=1000
     maxLinearVelocity=200
     velocityScaleZ=0
     externalLinearDragV=<1 1>
    }
   }
   n27 = "Transform (Modify) Bone - Bone: dyn_loin_belt_base"(ComponentPose=@n26, Alpha=Get thigh_rot_up_l(), Translation="0,0,0", Rotation="0,0,0", Scale="0.8,1,1") {
    rotationMode=BMM_Additive
    rotationSpace=BCS_BoneSpace
    scaleMode=BMM_Additive
    scaleSpace=BCS_BoneSpace
    translationMode=BMM_Additive
    translationSpace=BCS_BoneSpace
    alphaBoolBlend {blendOption=Linear}
    alphaScaleBiasClamp {interpSpeedDecreasing=5, interpSpeedIncreasing=5}
    boneToModify {boneName=dyn_loin_belt_base}
    scale {x=0.8}
   }
   n28 = "Transform (Modify) Bone - Bone: dyn_loin_belt_base"(ComponentPose=@n27, Alpha=Get thigh_rot_up_r(), Translation="0,0,0", Rotation="0,0,0", Scale="0.8,1,1") {
    rotationMode=BMM_Additive
    rotationSpace=BCS_BoneSpace
    scaleMode=BMM_Additive
    scaleSpace=BCS_BoneSpace
    translationMode=BMM_Additive
    translationSpace=BCS_BoneSpace
    alphaBoolBlend {blendOption=Linear}
    alphaScaleBiasClamp {interpSpeedDecreasing=5, interpSpeedIncreasing=5}
    boneToModify {boneName=dyn_loin_belt_base}
    scale {x=0.8}
   }
   Output Pose(Result=Component To Local(ComponentPose=@n28)) {layerGroup=None}
  ]}
 EventGraph {nodes [
   Event BlueprintUpdateAnimation()
   > n1 = Sequence()
   n2 = BreakTransform(InTransform=GetDeltaTransformFromRefPose(self=GetOwnerSkeletalMeshComponent(), BoneName=thigh_r, BaseName=pelvis))
   > Set thigh_rot_up_r(thigh_rot_up_r:double=MapRangeClamped(Value=BreakRotator(InRot=@n2.Rotation).Yaw, InRangeA=-15, InRangeB=-100)) -> (Output_Get:double)
   n3 = BreakTransform(InTransform=GetDeltaTransformFromRefPose(self=GetOwnerSkeletalMeshComponent(), BoneName=thigh_l, BaseName=pelvis))
   > Set thigh_rot_up_l(thigh_rot_up_l:double=MapRangeClamped(Value=BreakRotator(InRot=@n3.Rotation).Yaw, InRangeA=-10, InRangeB=-100)) -> (Output_Get:double)
   n4 = GetOwnerSkeletalMeshComponent()
   n5 = BreakRotator(InRot=BreakTransform(InTransform=GetDeltaTransformFromRefPose(self=@n4, BoneName=upperarm_r, BaseName=spine_05)).Rotation)
   > Set arm_rot_up_r(arm_rot_up_r:double=MapRangeClamped(Value="Absolute (Float)"(A=@n5.Yaw), InRangeA=80, InRangeB=145)) -> (Output_Get:double)
   n6 = GetOwnerSkeletalMeshComponent()
   n7 = BreakRotator(InRot=BreakTransform(InTransform=GetDeltaTransformFromRefPose(self=@n6, BoneName=upperarm_l, BaseName=spine_05)).Rotation)
   > Set arm_rot_up_l(arm_rot_up_l:double=MapRangeClamped(Value="Absolute (Float)"(A=@n7.Yaw), InRangeA=60, InRangeB=110)) -> (Output_Get:double)
   n8 = Gravity Override State Function DEPRECATED(
    execute=@n1.then_1
    CharacterPartAnimInstance=Self-Reference()
    GameTimeInSeconds=GetGameTimeInSeconds()
    GravityOverrideParams=Get Gravity Override Params()
   )
   > Set GravityOverride(GravityOverride=@n8)
  ]}
]
variables [
 arm_rot_up_l {type=number}
 arm_rot_up_r {type=number}
 Gravity Override Params {
  type=GravityOverrideParamsStruct
  default {skydive_FloatingDown {gravityOverride=<0 -700 100>, windAmplitude=<200 -500 200>, windOffset=<0 -600 200>}}
 }
 GravityOverride {type=Vector}
 thigh_rot_up_l {type=number}
 thigh_rot_up_r {type=number}
]
