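# Physics asset for the dynamic head and neck feathers on the Raven character's face accessory.
# Four capsule feathers per side of the head hang off left and right master bodies, while three neck feathers each have their own master.
# Masters attach to head_offset with free swing and twist, while each feather has limit angles set against its master.
# Upper-arm capsules have no constraints, apparently serving as colliders for the feathers.
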PhysicsAsset /BRCosmetics/Characters/Player/Male/Medium/Bodies/M_MED_Critter_Raven/Meshes/Parts/M_MED_Critter_Raven_FaceAcc_dyn_feathers
bodies {
 dyn_head_feathers_l_01 [{shapeType=Capsule, length=10, radius=2, shapeName=dyn_head_feathers_l_01_capsule, center=<4.7305 0 0.0004>, rotation=<rot 90 0 0>}]
 dyn_head_feathers_l_02 [{shapeType=Capsule, length=10, radius=2, shapeName=dyn_head_feathers_l_02_capsule, center=<3.735 0 -0.0001>, rotation=<rot 90 0 0>}]
 dyn_head_feathers_l_03 [{shapeType=Capsule, length=10, radius=2, shapeName=dyn_head_feathers_l_03_capsule, center=<2.5001 0 0.0001>, rotation=<rot 90 0 0>}]
 dyn_head_feathers_l_04 [{shapeType=Capsule, length=5.5, radius=2, shapeName=dyn_head_feathers_l_04_capsule, center=<0.8103 0 0.0002>, rotation=<rot 90 0 0>}]
 dyn_head_feathers_l_MSTR [{shapeType=Capsule, length=1, radius=1}]
 dyn_head_feathers_r_01 [{shapeType=Capsule, length=10, radius=2, shapeName=dyn_head_feathers_l_01_capsule, center=<4.7305 0 0.0004>, rotation=<rot 90 0 0>}]
 dyn_head_feathers_r_02 [{shapeType=Capsule, length=10, radius=2, shapeName=dyn_head_feathers_l_02_capsule, center=<3.735 0 -0.0001>, rotation=<rot 90 0 0>}]
 dyn_head_feathers_r_03 [{shapeType=Capsule, length=10, radius=2, shapeName=dyn_head_feathers_l_03_capsule, center=<2.5001 0 0.0001>, rotation=<rot 90 0 0>}]
 dyn_head_feathers_r_04 [{shapeType=Capsule, length=5.5, radius=2, shapeName=dyn_head_feathers_l_04_capsule, center=<0.8103 0 0.0002>, rotation=<rot 90 0 0>}]
 dyn_head_feathers_r_MSTR [{shapeType=Capsule, length=1, radius=1}]
 dyn_neck_feather_bk [{shapeType=Capsule, length=10, radius=4, shapeName=dyn_neck_feather_bk_capsule, center=<5 0 0>, rotation=<rot 0 0 90>}]
 dyn_neck_feather_bk_MSTR [{shapeType=Capsule, length=1, radius=1}]
 dyn_neck_feather_l [{shapeType=Capsule, length=10, radius=4, shapeName=dyn_neck_feather_l_capsule, center=<4.5 0 1>, rotation=<rot 0 0 90>}]
 dyn_neck_feather_l_MSTR [{shapeType=Capsule, length=1, radius=1}]
 dyn_neck_feather_r [{shapeType=Capsule, length=10, radius=4, shapeName=dyn_neck_feather_r_capsule, center=<4 0 1>, rotation=<rot -0.8671 9.963 94.9242>}]
 dyn_neck_feather_r_MSTR [{shapeType=Capsule, length=1, radius=1}]
 head_offset [{shapeType=Capsule, length=1, radius=1}]
 upperarm_l [{shapeType=Capsule, length=7.5, radius=5.75, shapeName=upperarm_r_capsule, center=<4.5235 0.3307 3.1778>, rotation=<rot 84.9997 0 180>}]
 upperarm_r [{shapeType=Capsule, length=7.5, radius=5.75, shapeName=upperarm_r_capsule, center=<-4.5831 -0.3307 -3.8587>, rotation=<rot 84.9997 0 0>}]
}
constraints [
 {bone1Name=dyn_neck_feather_r, bone2Name=dyn_neck_feather_r_MSTR, swing1LimitDegrees=10, swing2LimitDegrees=30, twistLimitDegrees=2}
 {bone1Name=dyn_neck_feather_l, bone2Name=dyn_neck_feather_l_MSTR, swing1LimitDegrees=10, swing2LimitDegrees=30, twistLimitDegrees=2}
 {bone1Name=dyn_neck_feather_bk, bone2Name=dyn_neck_feather_bk_MSTR, swing1LimitDegrees=10, swing2LimitDegrees=30, twistLimitDegrees=2}
 {
  bone1Name=dyn_head_feathers_l_MSTR
  bone2Name=head_offset
  swing1LimitDegrees=45
  swing1Motion=Free
  swing2LimitDegrees=45
  swing2Motion=Free
  twistLimitDegrees=45
  twistMotion=Free
 }
 {bone1Name=dyn_head_feathers_l_01, bone2Name=dyn_head_feathers_l_MSTR, swing1LimitDegrees=25, swing2LimitDegrees=15, twistLimitDegrees=5}
 {bone1Name=dyn_head_feathers_l_02, bone2Name=dyn_head_feathers_l_MSTR, swing1LimitDegrees=25, swing2LimitDegrees=15, twistLimitDegrees=5}
 {bone1Name=dyn_head_feathers_l_03, bone2Name=dyn_head_feathers_l_MSTR, swing1LimitDegrees=25, swing2LimitDegrees=15, twistLimitDegrees=5}
 {bone1Name=dyn_head_feathers_l_04, bone2Name=dyn_head_feathers_l_MSTR, swing1LimitDegrees=25, swing2LimitDegrees=15, twistLimitDegrees=5}
 {
  bone1Name=dyn_head_feathers_r_MSTR
  bone2Name=head_offset
  swing1LimitDegrees=45
  swing1Motion=Free
  swing2LimitDegrees=45
  swing2Motion=Free
  twistLimitDegrees=45
  twistMotion=Free
 }
 {bone1Name=dyn_head_feathers_r_01, bone2Name=dyn_head_feathers_r_MSTR, swing1LimitDegrees=25, swing2LimitDegrees=15, twistLimitDegrees=5}
 {bone1Name=dyn_head_feathers_r_02, bone2Name=dyn_head_feathers_r_MSTR, swing1LimitDegrees=25, swing2LimitDegrees=15, twistLimitDegrees=5}
 {bone1Name=dyn_head_feathers_r_03, bone2Name=dyn_head_feathers_r_MSTR, swing1LimitDegrees=25, swing2LimitDegrees=15, twistLimitDegrees=5}
 {bone1Name=dyn_head_feathers_r_04, bone2Name=dyn_head_feathers_r_MSTR, swing1LimitDegrees=25, swing2LimitDegrees=15, twistLimitDegrees=5}
 {
  bone1Name=dyn_neck_feather_r_MSTR
  bone2Name=head_offset
  swing1LimitDegrees=45
  swing1Motion=Free
  swing2LimitDegrees=45
  swing2Motion=Free
  twistLimitDegrees=45
  twistMotion=Free
 }
 {
  bone1Name=dyn_neck_feather_l_MSTR
  bone2Name=head_offset
  swing1LimitDegrees=45
  swing1Motion=Free
  swing2LimitDegrees=45
  swing2Motion=Free
  twistLimitDegrees=45
  twistMotion=Free
 }
 {
  bone1Name=dyn_neck_feather_bk_MSTR
  bone2Name=head_offset
  swing1LimitDegrees=45
  swing1Motion=Free
  swing2LimitDegrees=45
  swing2Motion=Free
  twistLimitDegrees=45
  twistMotion=Free
 }
]
