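# Animation Blueprint for the Raven critter skin's face accessory, driving its dynamic feathers.
# It layers the body mesh's arm branches over the attached parent's pose, then additively adjusts the right neck feather master bone by upper-arm yaw.
# Two neck-relative RigidBody nodes use separate physics assets, weighted inversely by the "Hide Element 01" material parameter read when customization loads.
# A deprecated state function updates their gravity from per-movement-state parameters.
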AnimBlueprint /BRCosmetics/Characters/Player/Male/Medium/Bodies/M_MED_Critter_Raven/Meshes/Parts/M_MED_Critter_Raven_FaceAcc_AnimBP
target_skeleton=&/M_MED_Critter_Raven_FaceAcc_Skeleton
graphs [
 AnimGraph {nodes [
   n1 = Copy Pose From Mesh(SourceMeshComponent=Get Body_Mesh())
   n2 = Copy Pose From Mesh() {bUseAttachedParent=true}
   n3 = "Transform (Modify) Bone - Bone: dyn_neck_feather_r_MSTR"(
    ComponentPose=Local To Component(LocalPose=Layered blend per bone(BasePose=@n2, BlendPoses_0=@n1) {layerSetup [{branchFilters [{boneName=upperarm_l}, {boneName=upperarm_r}]}]})
    Alpha=Get arm_rot_up_r() -> (arm_rot_up_r:double)
    Translation="-0.162,0.173,2.882"
    Rotation="20.000,-0.002,24.996"
    Scale="1,1,1"
   ) {
    rotationMode=BMM_Additive
    rotationSpace=BCS_BoneSpace
    translationMode=BMM_Additive
    translationSpace=BCS_BoneSpace
    alphaBoolBlend {blendOption=Linear}
    boneToModify {boneName=dyn_neck_feather_r_MSTR}
    rotation=<rot 20 -0.002 24.9958>
    translation=<-0.162 0.173 2.882>
   }
   n4 = Get GravityOverride()
   n5 = RigidBody!(ComponentPose=@n3, Alpha=Get wingsOffAlpha() -> (wingsOffAlpha:double), OverrideWorldGravity:Vector=@n4, ExternalForce) -> (Pose) {
    bClampLinearTranslationLimitToRefPose=true
    bOverrideWorldGravity=true
    overridePhysicsAsset=&/M_MED_Critter_Raven_FaceAcc_dyn_wingsOff
    simulationSpace=BaseBoneSpace
    alphaBoolBlend {blendOption=Linear}
    baseBoneRef {boneName=neck_01}
    componentAppliedLinearAccClamp=<500 500 500>
    componentLinearAccScale=<0.1 0.1 0.1>
    componentLinearVelScale=<0.1 0.1 0.1>
    overrideWorldGravity {z=-100}
   }
   n6 = RigidBody(ComponentPose=@n5, Alpha=Get wingsOnAlpha() -> (wingsOnAlpha:double), OverrideWorldGravity=@n4) {
    bClampLinearTranslationLimitToRefPose=true
    bOverrideWorldGravity=true
    overridePhysicsAsset=&/M_MED_Critter_Raven_FaceAcc_dyn_feathers
    simulationSpace=BaseBoneSpace
    alphaBoolBlend {blendOption=Linear}
    baseBoneRef {boneName=neck_01}
    componentAppliedLinearAccClamp=<500 500 500>
    componentLinearAccScale=<0.1 0.1 0.1>
    componentLinearVelScale=<0.1 0.1 0.1>
    overrideWorldGravity {z=-100}
   }
   Output Pose(Result=Component To Local(ComponentPose=@n6)) {layerGroup=None}
  ]}
 EventGraph {nodes [
   Event BlueprintUpdateAnimation()
   > n1 = Sequence!(execute) -> (then_0, then_1, then_2:Exec)
   > n2 = Gravity Override State Function DEPRECATED(CharacterPartAnimInstance=Self-Reference(), GameTimeInSeconds=GetGameTimeInSeconds(), GravityOverrideParams=Get Gravity Override Params())
   > Set GravityOverride(GravityOverride=@n2)
   n3 = GetOwnerSkeletalMeshComponent()
   n4 = BreakRotator(InRot=BreakTransform(InTransform=GetDeltaTransformFromRefPose(self=@n3, BoneName=upperarm_r, BaseName=spine_05)).Rotation)
   Set arm_rot_up_r(execute=@n1.then_1, arm_rot_up_r:double=MapRangeClamped(Value="Absolute (Float)"(A=@n4.Yaw), InRangeA=60, InRangeB=110)) -> (Output_Get:double)
   Event OnCharacterCustomizationLoaded()
   > n5 = Get Dynamic Material Instance(SkeletalMeshComponent=GetPartSkeletalMeshComponent())
   > n6 = Get Scalar Parameter Value!(execute, self:Material Instance Dynamic Object Reference=@n5, ParameterName:Name=Hide Element 01) -> (then, ReturnValue:float) {is=title K2Node_CallFunction}
   > n7 = Set wingsOffAlpha(wingsOffAlpha:double=@n6) -> (Output_Get:double)
   > Set wingsOnAlpha(wingsOnAlpha:double=MapRangeClamped(Value=@n7, InRangeA=0, InRangeB=1, OutRangeA=1, OutRangeB=0)) -> (Output_Get:double)
   Event BlueprintInitializeAnimation()
   > Set Body_Mesh(Body_Mesh=GetOwnerSkeletalMeshComponent())
  ]}
]
variables [
 arm_rot_up_r {type=number}
 Body_Mesh {type=SkeletalMeshComponent}
 Gravity Override Params {
  type=GravityOverrideParamsStruct
  default {
   emoteOrMelee {gravityOverride=<0 0 -100>}
   front_End {gravityOverride=<0 0 -100>}
   onGround_Moving {gravityOverride=<0 0 -100>}
   onGround_Standing {gravityOverride=<0 0 -100>}
   skydive_DivingDown {gravityOverride=<-500 0 0>}
   skydive_FloatingDown {gravityOverride=<0 -100 0>, windAmplitude=<20 -80 80>, windOffset=<0 -10 10>}
   skydive_Parachuting {gravityOverride=<0 0 -100>, windAmplitude=<20 50 50>}
   swimming {gravityOverride=<-100 0 0>}
  }
 }
 GravityOverride {type=Vector}
 wingsOffAlpha {type=number}
 wingsOnAlpha {type=number}
]
